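# Material function for the InchEscape character's anime inverted-hull outline, outputting material attributes and an FX map.
# World position offset pushes the hull along selectable normals, with thickness from masks, distance fade and optional noise.
# A line-boil pattern varies thickness, animating a procedural grid over bounds-normalized pre-skinned positions in stepped time.
# Two-sided sign masks out front faces; color passes through time-of-day and LUT adjustments into emissive.
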MaterialFunction /BRCosmetics/Characters/Player/Male/Medium/Bodies/M_MED_InchEscape/Materials/MFC_AnimeOutline_InvHull_InchEscape
graph {expressions {
  MaterialExpressionAbs_0 {class=MaterialExpressionAbs, inputs [None {from=MaterialExpressionSubtract_3, output=""}], properties {}}
  MaterialExpressionAbs_1 {class=MaterialExpressionAbs, inputs [None {from=MaterialExpressionDotProduct_0, output=""}], properties {}}
  MaterialExpressionAbs_3 {class=MaterialExpressionAbs, inputs [None {from=MaterialExpressionSubtract_8, output=""}], properties {}}
  MaterialExpressionAdd_0 {
   class=MaterialExpressionAdd
   inputs [
    A {from=MaterialExpressionLinearInterpolate_1}
    B {from=MaterialExpressionReroute_5}
   ]
  }
  MaterialExpressionAdd_2 {
   class=MaterialExpressionAdd
   inputs [
    A {from=MaterialExpressionNamedRerouteUsage_17}
    B {from=MaterialExpressionStaticSwitchParameter_12}
   ]
  }
  MaterialExpressionAdd_3 {
   class=MaterialExpressionAdd
   inputs [
    A {from=MaterialExpressionChannelMaskParameter_1, output=R}
    B {from=MaterialExpressionConstant_0}
   ]
  }
  MaterialExpressionAdd_4 {
   class=MaterialExpressionAdd
   inputs [
    A {from=MaterialExpressionMultiply_4}
    B {from=MaterialExpressionScalarParameter_6}
   ]
  }
  MaterialExpressionAdd_6 {
   class=MaterialExpressionAdd
   inputs [
    A {from=MaterialExpressionComponentMask_3}
    B {from=MaterialExpressionComponentMask_1}
   ]
  }
  MaterialExpressionCameraPositionWS_0 {class=MaterialExpressionCameraPositionWS, inputs [], properties {}}
  MaterialExpressionCameraVectorWS_0 {class=MaterialExpressionCameraVectorWS, inputs [], properties {}}
  MaterialExpressionCeil_0 {
   class=MaterialExpressionCeil
   inputs [None {from=MaterialExpressionNamedRerouteDeclaration_5, output=""}]
   properties {desc=cut outlines mask}
  }
  MaterialExpressionChannelMaskParameter_1 {
   class=MaterialExpressionChannelMaskParameter
   inputs [None {from=MaterialExpressionFunctionOutput_1, output=""}]
   properties {group=01_InvertedHull_Outlines, maskChannel=Blue, parameterName=OutlineThickness_Channel, sortPriority=3, defaultValue=<rgba 0 0 1 0>}
  }
  MaterialExpressionComponentMask_1 {class=MaterialExpressionComponentMask, inputs [None {from=MaterialExpressionPanner_1}], properties {g=true}}
  MaterialExpressionComponentMask_3 {class=MaterialExpressionComponentMask, inputs [None {from=MaterialExpressionPanner_1}], properties {r=true}}
  MaterialExpressionComponentMask_6 {class=MaterialExpressionComponentMask, inputs [None {from=MaterialExpressionNamedRerouteUsage_23}], properties {b=true, r=true}}
  MaterialExpressionComponentMask_7 {class=MaterialExpressionComponentMask, inputs [None {from=MaterialExpressionNamedRerouteUsage_23}], properties {r=true}}
  MaterialExpressionConstant2Vector_0 {class=MaterialExpressionConstant2Vector, properties {g=1, r=-1}}
  MaterialExpressionConstant_0 {class=MaterialExpressionConstant, properties {desc=one bit used for opacity cutout, r=-0.009}}
  MaterialExpressionConstant_1 {class=MaterialExpressionConstant, properties {r=1}}
  MaterialExpressionConstant_2 {class=MaterialExpressionConstant, properties {r=1}}
  MaterialExpressionConstant_6 {class=MaterialExpressionConstant, properties {r=1}}
  MaterialExpressionConvert_1 {
   class=MaterialExpressionConvert
   inputs [
    None {from=""}
    None {from=MaterialExpressionScalarParameter_9}
   ]
   properties {
    nodeName=Make Vector2
    convertInputs [
     {type=Scalar, expressionInput {expression=null, inputName=None, mask=0, maskA=0, maskB=0, maskG=0, maskR=0, outputIndex=0}}
     {
      type=Scalar
      expressionInput {
       expression=&self:MaterialExpressionScalarParameter_9
       inputName=None
       mask=0
       maskA=0
       maskB=0
       maskG=0
       maskR=0
       outputIndex=0
      }
     }
    ]
    convertOutputs [{type=Vector2}]
   }
  }
  MaterialExpressionDivide_0 {
   class=MaterialExpressionDivide
   inputs [
    A {from=MaterialExpressionVectorParameter_0}
    B {from=""}
   ]
   properties {constB=10}
  }
  MaterialExpressionDivide_1 {
   class=MaterialExpressionDivide
   inputs [
    A {from=MaterialExpressionFloor_0}
    B {from=""}
   ]
   properties {constB=3}
  }
  MaterialExpressionDivide_3 {
   class=MaterialExpressionDivide
   inputs [
    A {from=MaterialExpressionSubtract_4}
    B {from=MaterialExpressionPreSkinnedLocalBounds_0, output=Extents}
   ]
  }
  MaterialExpressionDotProduct_0 {
   class=MaterialExpressionDotProduct
   inputs [
    A {from=MaterialExpressionCameraVectorWS_0, output=""}
    B {from=MaterialExpressionTransform_2, output=""}
   ]
   properties {}
  }
  MaterialExpressionFloor_0 {class=MaterialExpressionFloor, inputs [None {from=MaterialExpressionMultiply_9, output=""}], properties {}}
  MaterialExpressionFrac_0 {class=MaterialExpressionFrac, inputs [None {from=MaterialExpressionMultiply_17, output=""}], properties {}}
  MaterialExpressionFrac_3 {class=MaterialExpressionFrac, inputs [None {from=MaterialExpressionMultiply_18, output=""}], properties {}}
  MaterialExpressionFunctionOutput_0 {class=MaterialExpressionFunctionOutput, inputs [None {from=MaterialExpressionStaticSwitchParameter_2, output=""}]}
  MaterialExpressionFunctionOutput_1 {
   class=MaterialExpressionFunctionOutput
   inputs [None {from=MaterialExpressionTextureSampleParameter2D_3, output=RGBA}]
   properties {outputName=FX Map, sortPriority=1}
  }
  MaterialExpressionLinearInterpolate_0 {
   class=MaterialExpressionLinearInterpolate
   inputs [
    A {from=MaterialExpressionReroute_0}
    B {from=MaterialExpressionMin_0}
    Alpha {from=MaterialExpressionPower_0}
   ]
  }
  MaterialExpressionLinearInterpolate_1 {
   class=MaterialExpressionLinearInterpolate
   inputs [
    A {from=""}
    B {from=MaterialExpressionMultiply_0}
    Alpha {from=MaterialExpressionStaticSwitchParameter_14}
   ]
   properties {constB=0}
  }
  MaterialExpressionLinearInterpolate_2 {
   class=MaterialExpressionLinearInterpolate
   inputs [
    A {from=MaterialExpressionScalarParameter_10}
    B {from=MaterialExpressionScalarParameter_36}
    Alpha {from=MaterialExpressionSaturate_0}
   ]
   properties {constA=0.3}
  }
  MaterialExpressionLinearInterpolate_5 {
   class=MaterialExpressionLinearInterpolate
   inputs [
    A {from=MaterialExpressionMaterialFunctionCall_6, output=Rotated Values}
    B {from=MaterialExpressionMultiply_14}
    Alpha {from=MaterialExpressionStep_2}
   ]
  }
  MaterialExpressionLinearInterpolate_8 {
   class=MaterialExpressionLinearInterpolate
   inputs [
    A {from=MaterialExpressionReroute_4}
    B {from=MaterialExpressionMultiply_11}
    Alpha {from=MaterialExpressionNamedRerouteUsage_0}
   ]
  }
  MaterialExpressionLocalPosition_0 {class=MaterialExpressionLocalPosition, inputs [], properties {localOrigin=InstancePreSkinning}}
  MaterialExpressionMaterialFunctionCall_0 {
   class=MaterialExpressionMaterialFunctionCall
   inputs [
    Fade Offset {from=MaterialExpressionScalarParameter_39}
    Fade Distance {from=MaterialExpressionScalarParameter_38}
   ]
   properties {materialFunction=/Game/Global/Materials/Character_Parent_2023/Anime_Parent_2023/MaterialFunctions/MFC_DistanceFade_FOV_compensation}
  }
  MaterialExpressionMaterialFunctionCall_1 {
   class=MaterialExpressionMaterialFunctionCall
   inputs [
    Normal {from=MaterialExpressionTextureSampleParameter2D_4, output=RGB}
    Flatness {from=MaterialExpressionScalarParameter_0}
   ]
   properties {materialFunction=&FlattenNormal}
  }
  MaterialExpressionMaterialFunctionCall_13 {
   class=MaterialExpressionMaterialFunctionCall
   inputs [
    In {from=MaterialExpressionMaterialFunctionCall_3, output=Result}
    LUT {from=""}
    LUT Multiplier {from=""}
   ]
   properties {materialFunction=/Game/Packages/Fortress_SharedMaterials/Anime_Master_Material/MF_ApplyLUT}
  }
  MaterialExpressionMaterialFunctionCall_2 {
   class=MaterialExpressionMaterialFunctionCall
   inputs [WPO {from=MaterialExpressionReroute_3}]
   properties {materialFunction=/Game/Packages/Fortress_SharedMaterials/Base_Material_Functions/Fortnite_HideElementsWithVertexColor_2022}
  }
  MaterialExpressionMaterialFunctionCall_25 {
   class=MaterialExpressionMaterialFunctionCall
   inputs [Color {from=MaterialExpressionMaterialFunctionCall_13, output=Result}]
   properties {materialFunction=/Game/Packages/Fortress_SharedMaterials/Anime_Master_Material/MF_TODColorAdjustment}
  }
  MaterialExpressionMaterialFunctionCall_3 {
   class=MaterialExpressionMaterialFunctionCall
   inputs [In {from=MaterialExpressionNamedRerouteUsage_3}]
   properties {materialFunction=/Game/Packages/Fortress_SharedMaterials/Anime_Master_Material/MF_TODCustomColors}
  }
  MaterialExpressionMaterialFunctionCall_4 {
   class=MaterialExpressionMaterialFunctionCall
   inputs []
   properties {materialFunction=/Game/Global/Materials/Character_Parent_2023/Anime_Parent_2023/MaterialFunctions/MFC_ActorAlighnedScreenPosition}
  }
  MaterialExpressionMaterialFunctionCall_6 {
   class=MaterialExpressionMaterialFunctionCall
   inputs [
    UVs {from=MaterialExpressionComponentMask_6}
    Rotation Center {from=MaterialExpressionStep_3}
    "Rotation Angle (0-1)" {from=MaterialExpressionMultiply_12}
   ]
   properties {materialFunction=&CustomRotator}
  }
  MaterialExpressionMax_2 {
   class=MaterialExpressionMax
   inputs [
    A {from=MaterialExpressionAbs_0}
    B {from=MaterialExpressionAbs_3}
   ]
  }
  MaterialExpressionMin_0 {
   class=MaterialExpressionMin
   inputs [
    A {from=MaterialExpressionReroute_0, output=""}
    B {from=MaterialExpressionNamedRerouteUsage_6, output=""}
   ]
   properties {}
  }
  MaterialExpressionMultiply_0 {
   class=MaterialExpressionMultiply
   inputs [
    A {from=MaterialExpressionNormalize_2}
    B {from=MaterialExpressionScalarParameter_2}
   ]
  }
  MaterialExpressionMultiply_1 {
   class=MaterialExpressionMultiply
   inputs [
    A {from=MaterialExpressionStaticSwitchParameter_4}
    B {from=MaterialExpressionNamedRerouteUsage_5}
   ]
  }
  MaterialExpressionMultiply_10 {
   class=MaterialExpressionMultiply
   inputs [
    A {from=MaterialExpressionVertexColor_0}
    B {from=""}
   ]
   properties {constB=2}
  }
  MaterialExpressionMultiply_11 {
   class=MaterialExpressionMultiply
   inputs [
    A {from=MaterialExpressionNamedRerouteUsage_21}
    B {from=MaterialExpressionMultiply_5}
   ]
   properties {constB=0.6}
  }
  MaterialExpressionMultiply_12 {
   class=MaterialExpressionMultiply
   inputs [
    A {from=MaterialExpressionDivide_1}
    B {from=MaterialExpressionScalarParameter_8}
   ]
  }
  MaterialExpressionMultiply_14 {
   class=MaterialExpressionMultiply
   inputs [
    A {from=MaterialExpressionMaterialFunctionCall_6, output=Rotated Values}
    B {from=MaterialExpressionConstant2Vector_0}
   ]
   properties {constB=-1}
  }
  MaterialExpressionMultiply_17 {
   class=MaterialExpressionMultiply
   inputs [
    A {from=MaterialExpressionSubtract_6}
    B {from=MaterialExpressionMultiply_23}
   ]
   properties {constB=10}
  }
  MaterialExpressionMultiply_18 {
   class=MaterialExpressionMultiply
   inputs [
    A {from=MaterialExpressionAdd_6}
    B {from=MaterialExpressionScalarParameter_7}
   ]
   properties {constB=10}
  }
  MaterialExpressionMultiply_19 {
   class=MaterialExpressionMultiply
   inputs [
    A {from=MaterialExpressionFrac_3}
    B {from=""}
   ]
   properties {constB=2}
  }
  MaterialExpressionMultiply_2 {
   class=MaterialExpressionMultiply
   inputs [
    A {from=MaterialExpressionCeil_0}
    B {from=MaterialExpressionOneMinus_1}
   ]
  }
  MaterialExpressionMultiply_21 {
   class=MaterialExpressionMultiply
   inputs [
    A {from=MaterialExpressionStaticSwitchParameter_6}
    B {from=MaterialExpressionNamedRerouteUsage_1}
   ]
  }
  MaterialExpressionMultiply_22 {
   class=MaterialExpressionMultiply
   inputs [
    A {from=MaterialExpressionTextureSampleParameter2D_0, output=RGB}
    B {from=MaterialExpressionScalarParameter_34}
   ]
  }
  MaterialExpressionMultiply_23 {
   class=MaterialExpressionMultiply
   inputs [
    A {from=MaterialExpressionScalarParameter_7}
    B {from=""}
   ]
   properties {constB=0.6667}
  }
  MaterialExpressionMultiply_3 {
   class=MaterialExpressionMultiply
   inputs [
    A {from=MaterialExpressionNamedRerouteUsage_16}
    B {from=MaterialExpressionScalarParameter_4}
   ]
   properties {constB=0.1}
  }
  MaterialExpressionMultiply_4 {
   class=MaterialExpressionMultiply
   inputs [
    A {from=MaterialExpressionSubtract_1}
    B {from=MaterialExpressionScalarParameter_5}
   ]
  }
  MaterialExpressionMultiply_5 {
   class=MaterialExpressionMultiply
   inputs [
    A {from=MaterialExpressionScalarParameter_40}
    B {from=""}
   ]
   properties {constB=2}
  }
  MaterialExpressionMultiply_6 {
   class=MaterialExpressionMultiply
   inputs [
    A {from=MaterialExpressionNamedRerouteUsage_7}
    B {from=MaterialExpressionStaticSwitchParameter_9}
   ]
  }
  MaterialExpressionMultiply_7 {
   class=MaterialExpressionMultiply
   inputs [
    A {from=MaterialExpressionFrac_0}
    B {from=""}
   ]
   properties {constB=2}
  }
  MaterialExpressionMultiply_9 {
   class=MaterialExpressionMultiply
   inputs [
    A {from=MaterialExpressionTime_0}
    B {from=""}
   ]
   properties {constB=12}
  }
  MaterialExpressionNamedRerouteDeclaration_0 {
   class=MaterialExpressionNamedRerouteDeclaration
   inputs [None {from=MaterialExpressionMaterialFunctionCall_0, output=Result}]
   properties {name=DistanceFading, nodeColor=<rgba 1 0.3047 0 1>}
  }
  MaterialExpressionNamedRerouteDeclaration_1 {
   class=MaterialExpressionNamedRerouteDeclaration
   inputs [None {from=MaterialExpressionPower_1}]
   properties {name=BakedNormalFresnel, nodeColor=<rgba 0 0.8144 1 1>}
  }
  MaterialExpressionNamedRerouteDeclaration_10 {
   class=MaterialExpressionNamedRerouteDeclaration
   inputs [None {from=MaterialExpressionStaticSwitchParameter_7}]
   properties {name=UVs, nodeColor=<rgba 1 0.6692 0 1>}
  }
  MaterialExpressionNamedRerouteDeclaration_11 {
   class=MaterialExpressionNamedRerouteDeclaration
   inputs [None {from=MaterialExpressionAdd_4}]
   properties {name=ThicknessNoise, nodeColor=<rgba 1 0.2972 0 1>}
  }
  MaterialExpressionNamedRerouteDeclaration_12 {
   class=MaterialExpressionNamedRerouteDeclaration
   inputs [None {from=MaterialExpressionLinearInterpolate_2}]
   properties {name=LineThickness, nodeColor=<rgba 0.3359 1 0 1>}
  }
  MaterialExpressionNamedRerouteDeclaration_13 {
   class=MaterialExpressionNamedRerouteDeclaration
   inputs [None {from=MaterialExpressionVertexColor_1, output=A}]
   properties {name=VertexColorAlpha, nodeColor=<rgba 0.8984 1 0 1>}
  }
  MaterialExpressionNamedRerouteDeclaration_15 {
   class=MaterialExpressionNamedRerouteDeclaration
   inputs [None {from=MaterialExpressionShaderStageSwitch_12}]
   properties {name=PreSkinnedLoc, nodeColor=<rgba 0.7001 0 1 1>}
  }
  MaterialExpressionNamedRerouteDeclaration_2 {
   class=MaterialExpressionNamedRerouteDeclaration
   inputs [None {from=MaterialExpressionStaticSwitchParameter_8}]
   properties {name=WPO, nodeColor=<rgba 1 0.3281 0 1>}
  }
  MaterialExpressionNamedRerouteDeclaration_20 {
   class=MaterialExpressionNamedRerouteDeclaration
   inputs [None {from=MaterialExpressionMaterialFunctionCall_4, output=UVs}]
   properties {name=ObjectAlighedScreenPosition, nodeColor=<rgba 1 0.867 0 1>}
  }
  MaterialExpressionNamedRerouteDeclaration_3 {
   class=MaterialExpressionNamedRerouteDeclaration
   inputs [None {from=MaterialExpressionMultiply_1}]
   properties {name=OpacityMask, nodeColor=<rgba 0 0.8125 1 1>}
  }
  MaterialExpressionNamedRerouteDeclaration_37 {
   class=MaterialExpressionNamedRerouteDeclaration
   inputs [None {from=MaterialExpressionDivide_3}]
   properties {name=Bounding, nodeColor=<rgba 0 0.5312 1 1>}
  }
  MaterialExpressionNamedRerouteDeclaration_4 {
   class=MaterialExpressionNamedRerouteDeclaration
   inputs [None {from=MaterialExpressionNormalize_0}]
   properties {name=BakedNormals, nodeColor=<rgba 0 0.2789 1 1>}
  }
  MaterialExpressionNamedRerouteDeclaration_5 {
   class=MaterialExpressionNamedRerouteDeclaration
   inputs [None {from=MaterialExpressionAdd_3}]
   properties {name=ThicknessMask, nodeColor=<rgba 1 0.1689 0 1>}
  }
  MaterialExpressionNamedRerouteDeclaration_6 {
   class=MaterialExpressionNamedRerouteDeclaration
   inputs [None {from=MaterialExpressionStaticSwitchParameter_0}]
   properties {name=Normal, nodeColor=<rgba 0.224 0.224 1 1>}
  }
  MaterialExpressionNamedRerouteDeclaration_7 {
   class=MaterialExpressionNamedRerouteDeclaration
   inputs [None {from=MaterialExpressionStaticSwitchParameter_11}]
   properties {name=ColorRaw, nodeColor=<rgba 0 0.5781 1 1>}
  }
  MaterialExpressionNamedRerouteDeclaration_8 {
   class=MaterialExpressionNamedRerouteDeclaration
   inputs [None {from=MaterialExpressionMaterialFunctionCall_25, output=Result}]
   properties {name=ColorFinal, nodeColor=<rgba 1 0.6797 0 1>}
  }
  MaterialExpressionNamedRerouteDeclaration_9 {
   class=MaterialExpressionNamedRerouteDeclaration
   inputs [None {from=MaterialExpressionStaticSwitchParameter_23}]
   properties {name=Cinematics OutlineThicknessTextureOverride, nodeColor=<rgba 1 0 0.5391 1>}
  }
  MaterialExpressionNamedRerouteUsage_0 {class=MaterialExpressionNamedRerouteUsage}
  MaterialExpressionNamedRerouteUsage_1 {class=MaterialExpressionNamedRerouteUsage}
  MaterialExpressionNamedRerouteUsage_10 {class=MaterialExpressionNamedRerouteUsage}
  MaterialExpressionNamedRerouteUsage_11 {class=MaterialExpressionNamedRerouteUsage}
  MaterialExpressionNamedRerouteUsage_12 {class=MaterialExpressionNamedRerouteUsage}
  MaterialExpressionNamedRerouteUsage_13 {class=MaterialExpressionNamedRerouteUsage}
  MaterialExpressionNamedRerouteUsage_14 {class=MaterialExpressionNamedRerouteUsage}
  MaterialExpressionNamedRerouteUsage_15 {class=MaterialExpressionNamedRerouteUsage}
  MaterialExpressionNamedRerouteUsage_16 {class=MaterialExpressionNamedRerouteUsage}
  MaterialExpressionNamedRerouteUsage_17 {class=MaterialExpressionNamedRerouteUsage}
  MaterialExpressionNamedRerouteUsage_18 {class=MaterialExpressionNamedRerouteUsage}
  MaterialExpressionNamedRerouteUsage_2 {class=MaterialExpressionNamedRerouteUsage}
  MaterialExpressionNamedRerouteUsage_20 {class=MaterialExpressionNamedRerouteUsage}
  MaterialExpressionNamedRerouteUsage_21 {class=MaterialExpressionNamedRerouteUsage}
  MaterialExpressionNamedRerouteUsage_23 {class=MaterialExpressionNamedRerouteUsage}
  MaterialExpressionNamedRerouteUsage_3 {class=MaterialExpressionNamedRerouteUsage}
  MaterialExpressionNamedRerouteUsage_4 {class=MaterialExpressionNamedRerouteUsage}
  MaterialExpressionNamedRerouteUsage_5 {class=MaterialExpressionNamedRerouteUsage}
  MaterialExpressionNamedRerouteUsage_6 {class=MaterialExpressionNamedRerouteUsage}
  MaterialExpressionNamedRerouteUsage_7 {class=MaterialExpressionNamedRerouteUsage}
  MaterialExpressionNamedRerouteUsage_8 {class=MaterialExpressionNamedRerouteUsage}
  MaterialExpressionNamedRerouteUsage_9 {class=MaterialExpressionNamedRerouteUsage}
  MaterialExpressionNormalize_0 {class=MaterialExpressionNormalize, inputs [VectorInput {from=MaterialExpressionSubtract_2, output=""}]}
  MaterialExpressionNormalize_1 {class=MaterialExpressionNormalize, inputs [VectorInput {from=MaterialExpressionMaterialFunctionCall_1, output=Result}]}
  MaterialExpressionNormalize_2 {class=MaterialExpressionNormalize, inputs [VectorInput {from=MaterialExpressionSubtract_0, output=""}]}
  MaterialExpressionOneMinus_1 {class=MaterialExpressionOneMinus, inputs [None {from=MaterialExpressionTwoSidedSign_0}]}
  MaterialExpressionOneMinus_2 {class=MaterialExpressionOneMinus, inputs [None {from=MaterialExpressionAbs_1}]}
  MaterialExpressionOneMinus_3 {class=MaterialExpressionOneMinus, inputs [None {from=MaterialExpressionTextureSampleParameter2D_1, output=R}]}
  MaterialExpressionPanner_0 {
   class=MaterialExpressionPanner
   inputs [
    Coordinate {from=MaterialExpressionMultiply_3}
    Time {from=""}
    Speed {from=MaterialExpressionDivide_0}
   ]
  }
  MaterialExpressionPanner_1 {
   class=MaterialExpressionPanner
   inputs [
    Coordinate {from=MaterialExpressionLinearInterpolate_5}
    Time {from=MaterialExpressionDivide_1}
    Speed {from=MaterialExpressionConvert_1}
   ]
   properties {speedY=-0.2}
  }
  MaterialExpressionPower_0 {
   class=MaterialExpressionPower
   inputs [
    Base {from=MaterialExpressionNamedRerouteDeclaration_13}
    Exp {from=MaterialExpressionScalarParameter_3}
   ]
  }
  MaterialExpressionPower_1 {
   class=MaterialExpressionPower
   inputs [
    Base {from=MaterialExpressionOneMinus_2}
    Exp {from=MaterialExpressionScalarParameter_1}
   ]
  }
  MaterialExpressionPower_2 {
   class=MaterialExpressionPower
   inputs [
    Base {from=MaterialExpressionMax_2}
    Exp {from=""}
   ]
   properties {constExponent=4}
  }
  MaterialExpressionPreSkinnedLocalBounds_0 {class=MaterialExpressionPreSkinnedLocalBounds, inputs [], properties {}}
  MaterialExpressionReroute_0 {class=MaterialExpressionReroute, inputs [None {from=MaterialExpressionStaticSwitchParameter_3}]}
  MaterialExpressionReroute_1 {class=MaterialExpressionReroute, inputs [None {from=MaterialExpressionOneMinus_1}]}
  MaterialExpressionReroute_3 {class=MaterialExpressionReroute, inputs [None {from=MaterialExpressionMultiply_21}]}
  MaterialExpressionReroute_4 {class=MaterialExpressionReroute, inputs [None {from=MaterialExpressionNamedRerouteUsage_21}]}
  MaterialExpressionReroute_5 {class=MaterialExpressionReroute, inputs [None {from=MaterialExpressionStaticSwitchParameter_1}]}
  MaterialExpressionSaturate_0 {class=MaterialExpressionSaturate, inputs [None {from=MaterialExpressionPower_2}]}
  MaterialExpressionScalarParameter_0 {
   class=MaterialExpressionScalarParameter
   properties {
    defaultValue=-0.1
    desc="values above 1.0 and below 0.0 can help with seperated geometry.\r\n-0.1 seems like a good wellyou to alight split geo."
    group=01_InvertedHull_Outlines
    parameterName=NormalMap_Flatness
    sliderMax=2
    sliderMin=-1
    sortPriority=5
   }
  }
  MaterialExpressionScalarParameter_1 {
   class=MaterialExpressionScalarParameter
   properties {defaultValue=5, group=01_InvertedHull_Outlines, parameterName=BakedNormalOpacity_Exponent, sliderMax=12}
  }
  MaterialExpressionScalarParameter_10 {
   class=MaterialExpressionScalarParameter
   properties {
    defaultValue=0.6
    desc="LineThickness (Near)"
    group=01_InvertedHull_Outlines
    parameterName=LineThickness_min
    sliderMax=0.5
    sortPriority=3
   }
  }
  MaterialExpressionScalarParameter_2 {
   class=MaterialExpressionScalarParameter
   properties {defaultValue=0.3, group=01_InvertedHull_Outlines, parameterName=CameraDirectionOffset, sliderMax=2, sortPriority=3}
  }
  MaterialExpressionScalarParameter_3 {
   class=MaterialExpressionScalarParameter
   properties {
    defaultValue=1
    desc="Controls the exponent curve of the alpha channel to mask out the Baked NormalOpacity Fresnel effect with the Vertex Color alpha channel. \r\nto improve the masking if the alpha channel is used for opacity masking, because of smooth normal for the hair."
    group=01_InvertedHull_Outlines
    parameterName=BakedNormalOpacity_AlphaExponent
    sliderMax=8
   }
  }
  MaterialExpressionScalarParameter_34 {
   class=MaterialExpressionScalarParameter
   properties {
    defaultValue=1
    desc=Multiplier for the Outline Color Texture
    group=01_InvertedHull_Outlines
    parameterName=Color_Intensity
    sliderMax=1
    sortPriority=3
   }
  }
  MaterialExpressionScalarParameter_36 {
   class=MaterialExpressionScalarParameter
   properties {
    defaultValue=1.2
    desc="LineThickness (Near)"
    group=01_InvertedHull_Outlines
    parameterName=LineThickness_max
    sliderMax=0.5
    sortPriority=3
   }
  }
  MaterialExpressionScalarParameter_38 {
   class=MaterialExpressionScalarParameter
   properties {defaultValue=1000, group=01_InvertedHull_Outlines, parameterName=Fade_Distance, sliderMax=1500, sliderMin=500, sortPriority=5}
  }
  MaterialExpressionScalarParameter_39 {
   class=MaterialExpressionScalarParameter
   properties {defaultValue=50, group=01_InvertedHull_Outlines, parameterName=Fade_Offset, sliderMax=100, sortPriority=5}
  }
  MaterialExpressionScalarParameter_4 {
   class=MaterialExpressionScalarParameter
   properties {defaultValue=0.2, group=01_InvertedHull_Outlines, parameterName=ThicknessVariation_Tiling}
  }
  MaterialExpressionScalarParameter_40 {
   class=MaterialExpressionScalarParameter
   properties {defaultValue=0.6, group=01_InvertedHull_Outlines, parameterName=LineThickness_Far, sliderMax=2, sortPriority=3}
  }
  MaterialExpressionScalarParameter_5 {
   class=MaterialExpressionScalarParameter
   properties {defaultValue=1, group=01_InvertedHull_Outlines, parameterName=ThicknessVariation_Intensity, sliderMax=2}
  }
  MaterialExpressionScalarParameter_6 {
   class=MaterialExpressionScalarParameter
   properties {group=01_InvertedHull_Outlines, parameterName=ThicknessVariation_Bias, sliderMax=1, sliderMin=-1}
  }
  MaterialExpressionScalarParameter_7 {
   class=MaterialExpressionScalarParameter
   properties {defaultValue=6, group=01_InvertedHull_Outlines, parameterName=LineBoilingNoiseDensity, sortPriority=3}
  }
  MaterialExpressionScalarParameter_8 {
   class=MaterialExpressionScalarParameter
   properties {
    defaultValue=1
    desc="LineThickness (Near)"
    group=01_InvertedHull_Outlines
    parameterName=LineBoilingNoiseRotateSpeed
    sliderMax=0.5
    sortPriority=3
   }
  }
  MaterialExpressionScalarParameter_9 {
   class=MaterialExpressionScalarParameter
   properties {defaultValue=0.2, desc="LineThickness (Near)", group=01_InvertedHull_Outlines, parameterName=LineBoilingSpeed, sliderMax=0.5, sortPriority=3}
  }
  MaterialExpressionSetMaterialAttributes_0 {
   class=MaterialExpressionSetMaterialAttributes
   inputs [
    MaterialAttributes {from="", output=""}
    EmissiveColor {from=MaterialExpressionNamedRerouteUsage_4, output=""}
    OpacityMask {from=MaterialExpressionNamedRerouteUsage_9, output=""}
    WorldPositionOffset {from=MaterialExpressionNamedRerouteUsage_10, output=""}
   ]
   properties {attributeSetTypes [#1, #2, #3]}
  }
  MaterialExpressionSetMaterialAttributes_1 {
   class=MaterialExpressionSetMaterialAttributes
   inputs [
    MaterialAttributes {from="", output=""}
    EmissiveColor {from=MaterialExpressionNamedRerouteUsage_2, output=""}
    OpacityMask {from="", output=""}
    WorldPositionOffset {from=MaterialExpressionNamedRerouteUsage_8, output=""}
   ]
   properties {
    desc=Debug
    attributeSetTypes [#1, #2, #3]
   }
  }
  MaterialExpressionShaderStageSwitch_12 {
   class=MaterialExpressionShaderStageSwitch
   inputs [
    PixelShader {from=MaterialExpressionVertexInterpolator_16, output=PS}
    VertexShader {from=MaterialExpressionLocalPosition_0, output=XYZ}
   ]
   properties {}
  }
  MaterialExpressionStaticSwitchParameter_0 {
   class=MaterialExpressionStaticSwitchParameter
   inputs [
    True {from=MaterialExpressionTransform_0, output=""}
    False {from=MaterialExpressionStaticSwitchParameter_5, output=""}
   ]
   properties {desc=use normal map to prevent segmented geometry, group=01_InvertedHull_Outlines, parameterName=UseNormalMap, sortPriority=0}
  }
  MaterialExpressionStaticSwitchParameter_1 {
   class=MaterialExpressionStaticSwitchParameter
   inputs [
    True {from=MaterialExpressionMaterialFunctionCall_2, output=Result}
    False {from=MaterialExpressionReroute_3, output=""}
   ]
   properties {group=03_Opacity Mask, parameterName=Use Vertex Colors for Mask}
  }
  MaterialExpressionStaticSwitchParameter_11 {
   class=MaterialExpressionStaticSwitchParameter
   inputs [
    True {from=MaterialExpressionMultiply_22, output=""}
    False {from=MaterialExpressionVectorParameter_4, output=RGB}
   ]
   properties {
    desc=Enable to use a texture to color the edge outlines of the mesh.
    group=01_InvertedHull_Outlines
    parameterName=UseDFL_ColorMap
    sortPriority=0
   }
  }
  MaterialExpressionStaticSwitchParameter_12 {
   class=MaterialExpressionStaticSwitchParameter
   inputs [
    True {from=MaterialExpressionLinearInterpolate_8, output=""}
    False {from=MaterialExpressionNamedRerouteUsage_21, output=""}
   ]
   properties {defaultValue=true, group=01_InvertedHull_Outlines, parameterName=UseDistanceFading, sortPriority=0}
  }
  MaterialExpressionStaticSwitchParameter_14 {
   class=MaterialExpressionStaticSwitchParameter
   inputs [
    True {from=MaterialExpressionNamedRerouteUsage_20, output=""}
    False {from=MaterialExpressionStaticSwitchParameter_15, output=""}
   ]
   properties {group=01_InvertedHull_Outlines, parameterName=UseOutlineOffsetVertexAlpha, sortPriority=0}
  }
  MaterialExpressionStaticSwitchParameter_15 {
   class=MaterialExpressionStaticSwitchParameter
   inputs [
    True {from=MaterialExpressionTextureSampleParameter2D_5, output=Mask}
    False {from=MaterialExpressionConstant_1, output=""}
   ]
   properties {group=01_InvertedHull_Outlines, parameterName=UseOutlineOffsetMap, sortPriority=0}
  }
  MaterialExpressionStaticSwitchParameter_2 {
   class=MaterialExpressionStaticSwitchParameter
   inputs [
    True {from=MaterialExpressionSetMaterialAttributes_1, output=""}
    False {from=MaterialExpressionSetMaterialAttributes_0, output=""}
   ]
   properties {group=01_InvertedHull_Outlines, parameterName=DebugThicknessMap, sortPriority=90}
  }
  MaterialExpressionStaticSwitchParameter_23 {
   class=MaterialExpressionStaticSwitchParameter
   inputs [
    True {from=MaterialExpressionOneMinus_3, output=""}
    False {from=MaterialExpressionConstant_6, output=""}
   ]
   properties {group=Cinematic, parameterName=UseCinematicsOutlineThicknessOverrideTexture}
  }
  MaterialExpressionStaticSwitchParameter_3 {
   class=MaterialExpressionStaticSwitchParameter
   inputs [
    True {from=MaterialExpressionMultiply_2, output=""}
    False {from=MaterialExpressionReroute_1, output=""}
   ]
   properties {group=01_InvertedHull_Outlines, parameterName=UseThicknessMap, sortPriority=0}
  }
  MaterialExpressionStaticSwitchParameter_4 {
   class=MaterialExpressionStaticSwitchParameter
   inputs [
    True {from=MaterialExpressionLinearInterpolate_0, output=""}
    False {from=MaterialExpressionStaticSwitchParameter_3, output=""}
   ]
   properties {
    desc=mask out this effect with the vertec color alpha channel
    group=01_InvertedHull_Outlines
    parameterName=UseBakedNormalOpacityFresnel
    sortPriority=0
   }
  }
  MaterialExpressionStaticSwitchParameter_5 {
   class=MaterialExpressionStaticSwitchParameter
   inputs [
    True {from=MaterialExpressionTransform_1, output=""}
    False {from=MaterialExpressionVertexNormalWS_0, output=""}
   ]
   properties {
    desc=use baked normals in the vertex colors instead of the surface normals.
    group=01_InvertedHull_Outlines
    parameterName=UseBakedNormals
    sortPriority=0
   }
  }
  MaterialExpressionStaticSwitchParameter_6 {
   class=MaterialExpressionStaticSwitchParameter
   inputs [
    True {from=MaterialExpressionMultiply_6, output=""}
    False {from=MaterialExpressionStaticSwitchParameter_9, output=""}
   ]
   properties {group=01_InvertedHull_Outlines, parameterName=UseThicknessMap, sortPriority=0}
  }
  MaterialExpressionStaticSwitchParameter_7 {
   class=MaterialExpressionStaticSwitchParameter
   inputs [
    True {from=MaterialExpressionTextureCoordinate_1, output=""}
    False {from=MaterialExpressionTextureCoordinate_0, output=""}
   ]
   properties {
    desc="active to use the second UV channel (UV1) for the Textures"
    group=01_InvertedHull_Outlines
    parameterName=UseSecondUVChannel
    sortPriority=0
   }
  }
  MaterialExpressionStaticSwitchParameter_8 {
   class=MaterialExpressionStaticSwitchParameter
   inputs [
    True {from=MaterialExpressionAdd_0, output=""}
    False {from=MaterialExpressionReroute_5, output=""}
   ]
   properties {group=01_InvertedHull_Outlines, parameterName=OutlineOffsetInCameraDirection, sortPriority=0}
  }
  MaterialExpressionStaticSwitchParameter_9 {
   class=MaterialExpressionStaticSwitchParameter
   inputs [
    True {from=MaterialExpressionAdd_2, output=""}
    False {from=MaterialExpressionStaticSwitchParameter_12, output=""}
   ]
   properties {group=01_InvertedHull_Outlines, parameterName=UseThicknessVariation, sortPriority=0}
  }
  MaterialExpressionStep_2 {
   class=MaterialExpressionStep
   inputs [
    Y {from=MaterialExpressionComponentMask_7, output=""}
    X {from="", output=""}
   ]
   properties {constX=0.5}
  }
  MaterialExpressionStep_3 {
   class=MaterialExpressionStep
   inputs [
    Y {from=MaterialExpressionComponentMask_6, output=""}
    X {from="", output=""}
   ]
   properties {constX=0.5}
  }
  MaterialExpressionSubtract_0 {
   class=MaterialExpressionSubtract
   inputs [
    A {from=MaterialExpressionCameraPositionWS_0}
    B {from=MaterialExpressionWorldPosition_0, output=XYZ}
   ]
  }
  MaterialExpressionSubtract_1 {
   class=MaterialExpressionSubtract
   inputs [
    A {from=MaterialExpressionTextureSample_0, output=R}
    B {from=""}
   ]
   properties {constB=0.5}
  }
  MaterialExpressionSubtract_2 {
   class=MaterialExpressionSubtract
   inputs [
    A {from=MaterialExpressionMultiply_10}
    B {from=""}
   ]
  }
  MaterialExpressionSubtract_3 {
   class=MaterialExpressionSubtract
   inputs [
    A {from=MaterialExpressionMultiply_7}
    B {from=""}
   ]
  }
  MaterialExpressionSubtract_4 {
   class=MaterialExpressionSubtract
   inputs [
    A {from=MaterialExpressionTransformPosition_4}
    B {from=MaterialExpressionPreSkinnedLocalBounds_0, output=Min}
   ]
  }
  MaterialExpressionSubtract_6 {
   class=MaterialExpressionSubtract
   inputs [
    A {from=MaterialExpressionComponentMask_3}
    B {from=MaterialExpressionComponentMask_1}
   ]
  }
  MaterialExpressionSubtract_8 {
   class=MaterialExpressionSubtract
   inputs [
    A {from=MaterialExpressionMultiply_19}
    B {from=""}
   ]
  }
  MaterialExpressionTextureCoordinate_0 {class=MaterialExpressionTextureCoordinate}
  MaterialExpressionTextureCoordinate_1 {class=MaterialExpressionTextureCoordinate, properties {coordinateIndex=1}}
  MaterialExpressionTextureSampleParameter2D_0 {
   class=MaterialExpressionTextureSampleParameter2D
   inputs [
    UVs {from=MaterialExpressionNamedRerouteUsage_14, output=""}
    Apply View MipBias {from="", output=""}
   ]
   properties {
    desc="Outline Color Texture, use can use the ShadedMap."
    group=01_InvertedHull_Outlines
    parameterName=DFL_Color_Map
    sortPriority=1
    texture=&T_BlacksRGB
   }
  }
  MaterialExpressionTextureSampleParameter2D_1 {
   class=MaterialExpressionTextureSampleParameter2D
   inputs [
    UVs {from=MaterialExpressionNamedRerouteUsage_15, output=""}
    Apply View MipBias {from="", output=""}
   ]
   properties {
    group=Cinematic
    parameterName=CinematicsOutlineThicknessOverride
    samplerType=SAMPLERTYPE_Masks
    texture=/Game/Global/Textures/Default/Blanks/LargeEmpty
   }
  }
  MaterialExpressionTextureSampleParameter2D_3 {
   class=MaterialExpressionTextureSampleParameter2D
   inputs [
    UVs {from=MaterialExpressionNamedRerouteUsage_12, output=""}
    Apply View MipBias {from="", output=""}
   ]
   properties {
    group=01_InvertedHull_Outlines
    parameterName=STT_Map
    samplerType=SAMPLERTYPE_Masks
    sortPriority=1
    texture=/Game/Packages/Fortress_SharedMaterials/Anime_Master_Material/T_Outline_Mask
    channelNames {a="", b=Thickness, g=Threshold, r=Sheen}
   }
  }
  MaterialExpressionTextureSampleParameter2D_4 {
   class=MaterialExpressionTextureSampleParameter2D
   inputs [
    UVs {from=MaterialExpressionNamedRerouteUsage_13, output=""}
    Apply View MipBias {from="", output=""}
   ]
   properties {
    group=01_InvertedHull_Outlines
    parameterName=Normal_Map
    samplerType=SAMPLERTYPE_Normal
    sortPriority=1
    texture=&FlatNormalBlue
   }
  }
  MaterialExpressionTextureSampleParameter2D_5 {
   class=MaterialExpressionTextureSampleParameter2D
   inputs [
    UVs {from="", output=""}
    Apply View MipBias {from="", output=""}
   ]
   properties {
    group=01_InvertedHull_Outlines
    parameterName=OutlineOffset_Map
    sortPriority=1
    texture=&T_WhitesRGB
    channelNames {a="", b="", g="", r=Mask}
   }
  }
  MaterialExpressionTextureSample_0 {
   class=MaterialExpressionTextureSample
   inputs [
    UVs {from=MaterialExpressionPanner_0}
    Tex {from=""}
    Apply View MipBias {from=""}
   ]
   properties {samplerSource=SSM_Wrap_WorldGroupSettings, samplerType=SAMPLERTYPE_LinearColor, texture=/EpicBaseTextures/Noises/T_Perlin_Noise_D}
  }
  MaterialExpressionTime_0 {class=MaterialExpressionTime, properties {bIgnorePause=true, bOverride_Period=true, period=1}}
  MaterialExpressionTransformPosition_4 {
   class=MaterialExpressionTransformPosition
   inputs [None {from=MaterialExpressionNamedRerouteUsage_18, output=""}]
   properties {transformType=TRANSFORMPOSSOURCE_Instance}
  }
  MaterialExpressionTransform_0 {class=MaterialExpressionTransform, inputs [None {from=MaterialExpressionNormalize_1}]}
  MaterialExpressionTransform_1 {
   class=MaterialExpressionTransform
   inputs [None {from=MaterialExpressionNamedRerouteDeclaration_4}]
   properties {transformSourceType=TRANSFORMSOURCE_Local}
  }
  MaterialExpressionTransform_2 {
   class=MaterialExpressionTransform
   inputs [None {from=MaterialExpressionNamedRerouteUsage_11}]
   properties {transformSourceType=TRANSFORMSOURCE_Local}
  }
  MaterialExpressionTwoSidedSign_0 {class=MaterialExpressionTwoSidedSign, inputs [], properties {}}
  MaterialExpressionVectorParameter_0 {
   class=MaterialExpressionVectorParameter
   properties {group=01_InvertedHull_Outlines, parameterName=ThicknessVariation_PannerSpeed, channelNames {a="", b="", g=Speed V, r=Speed U}}
  }
  MaterialExpressionVectorParameter_4 {
   class=MaterialExpressionVectorParameter
   properties {group=01_InvertedHull_Outlines, parameterName=Color, sortPriority=2, defaultValue=<rgba 0.005 0.005 0.005 1>}
  }
  MaterialExpressionVertexColor_0 {class=MaterialExpressionVertexColor, inputs [], properties {}}
  MaterialExpressionVertexColor_1 {class=MaterialExpressionVertexColor, inputs [], properties {}}
  MaterialExpressionVertexInterpolator_16 {class=MaterialExpressionVertexInterpolator, inputs [VS {from=MaterialExpressionLocalPosition_0, output=XYZ}], properties {}}
  MaterialExpressionVertexNormalWS_0 {class=MaterialExpressionVertexNormalWS, inputs [], properties {}}
  MaterialExpressionWorldPosition_0 {class=MaterialExpressionWorldPosition, inputs [], properties {}}
 }}
