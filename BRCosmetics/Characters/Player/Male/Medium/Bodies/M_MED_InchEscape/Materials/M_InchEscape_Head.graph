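# Head material for the InchEscape male medium character body, using Cloth shading with an anime look and animated flipbook face.
# Flipbook face color blends into the anime shader's lit and shaded colors, passes through optionally masked eye animation, then enters the shader's post stage.
# The flipbook's world position offset is added to the shader's.
# Temporal responsiveness varies within two rerouted masks when pose or look direction changes.
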Material /BRCosmetics/Characters/Player/Male/Medium/Bodies/M_MED_InchEscape/Materials/M_InchEscape_Head
graph {
 properties {
  bAutomaticallySetUsageInEditor=false
  bUseMaterialAttributes=true
  bUsedWithClothing=true
  bUsedWithMeshDeformer=true
  bUsedWithMorphTargets=true
  bUsedWithSkeletalMesh=true
  bUsedWithStaticLighting=true
  pixelDepthOffsetMode=PDOM_Legacy
  shadingModel=MSM_Cloth
  usedShadingModels=Cloth
 }
 nodes [
  n1 = MaterialFunctionCall!(Mouth UV, Eye UV, Brow UV, Material) {materialFunction=/Game/Global/Materials/Character_Parent_2023/FlipbookFace_Parent_2023/MaterialFunctions/MFI_FlipbookFace_Anime}
  NamedRerouteDeclaration(None=@n1.BrowsAlpha) {name=BrowsAlpha, nodeColor=<rgba 0.4531 0 1 1>}
  n2 = MaterialFunctionCall!(CustomUV) {materialFunction=/Game/Global/Materials/Character_Parent_2023/Anime_Parent_2023/MaterialFunctions/MFC_AnimeShader_2023_Pre}
  n3 = MaterialFunctionCall!(CustomLightVector, OverrideNormal_WS, STT=@n2.STT, Normal_Map=@n2.STT, DFL=@n2.STT, FX=@n2.STT) {materialFunction=/Game/Global/Materials/Character_Parent_2023/Anime_Parent_2023/MaterialFunctions/MFC_AnimeShader_2023_Calculations}
  n4 = NamedRerouteDeclaration(None=@n1.FaceAlpha) {name=FlipbookFaceAlpha, nodeColor=<rgba 0.8281 0 1 1>}
  n5 = ScalarParameter() {defaultValue=1, parameterName=FlipbookFaceAnimeShadedValue}
  n6 = VertexColor()
  n7 = TextureSampleParameter2D() {
   group=01_EyeAnimation_Parameters
   parameterName=EyeMask
   samplerSource=SSM_Wrap_WorldGroupSettings
   samplerType=SAMPLERTYPE_Masks
   sortPriority=-32
   texture=&T_WhiteMask
  }
  n8 = StaticSwitchParameter(True=AppendVector(A=@n6, B=@n6), False=@n7.RGBA) {group=01_EyeAnimation_Parameters, parameterName=UseVertexColorForEyeMask, sortPriority=-33}
  n9 = DotProduct(A=Constant3Vector() {constant=<rgba 1 1 1 0>}, B=VertexColor())
  n10 = Constant() {r=3}
  n11 = Constant()
  n12 = Constant() {r=1}
  n13 = MaterialFunctionCall!(
   Color Lit=LinearInterpolate(A=@n2.Color_Lit, B=@n1.FaceColor, Alpha=@n4)
   Color Shaded=LinearInterpolate(A=@n2.Color_Shaded, B=Multiply(A=@n1.FaceColor, B=@n5), Alpha=@n4)
   Use UV2=StaticBoolParameter() {defaultValue=true, group=01_EyeAnimation_Parameters, parameterName=Use UV2 for Eye Animation, sortPriority=-30}
   Mask=StaticSwitchParameter(
    True=ChannelMaskParameter(None=@n8) {group=01_EyeAnimation_Parameters, parameterName=EyeMaskChannel, sortPriority=-32}.R
    False=If(A=@n9, B=@n10, "A > B"=@n11, "A == B"=@n11, "A < B"=@n12)
   ) {group=01_EyeAnimation_Parameters, parameterName=UseEyeMask, sortPriority=-33}
  ) {materialFunction=/Game/Global/Materials/Character_Parent_2023/Anime_Parent_2023/MaterialFunctions/MF_AnimeEyeAnimation_Simple}
  n14 = GetMaterialAttributes(None=MaterialFunctionCall!(
    Shading_Gradient=@n3.Shading_Gradient
    Threshold_Gradient=@n3.Shading_Gradient
    Specular_Gradient=@n3.Shading_Gradient
    Sheen_Mask=@n3.Shading_Gradient
    Thickness=@n3.Shading_Gradient
    DFL_Mask=@n3.Shading_Gradient
    FX=@n3.Shading_Gradient
    Metallic_Gradient=@n3.Shading_Gradient
    RimLight_Gradient=@n3.Shading_Gradient
    NormalBuffer=@n3.Shading_Gradient
    --------
    CINE_LightSources=@n3.Shading_Gradient
    CINE_LightColor=@n3.Shading_Gradient
    ---------
    Color_Lit=@n13.Lit
    Color_Shaded=@n13.Lit
    SRM=@n2.SRM
    Emissive=@n2.SRM
    DFL_Color=@n2.SRM
    Specular_Color=@n2.SRM
    RimLight_Color=@n2.SRM
   ) {materialFunction=/Game/Global/Materials/Character_Parent_2023/Anime_Parent_2023/MaterialFunctions/MFC_AnimeShader_2023_Post}.Result) {attributeGetTypes [#1]}
  SetMaterialAttributes!(
   MaterialAttributes=@n14.MaterialAttributes
   World Position Offset=Add(
    A=@n14.World Position Offset
    B=GetMaterialAttributes(None=@n1.Material) {attributeGetTypes [#1]}.World Position Offset
   )
  ) {attributeSetTypes [#1]}
  n15 = Max(
   A=ScalarParameter() {group=Temporal Responsiveness, parameterName=HasPoseChanged}
   B=ScalarParameter() {group=Temporal Responsiveness, parameterName=HasLookDirectionChanged}
  )
  n16 = NamedRerouteUsage()
  n17 = NamedRerouteUsage()
  TemporalResponsivenessOutput!(None=LinearInterpolate(B=LinearInterpolate(Alpha=@n15) {constA=0.4}, Alpha=Max(A=@n16, B=@n17)) {constA=0.4})
 ]
}
