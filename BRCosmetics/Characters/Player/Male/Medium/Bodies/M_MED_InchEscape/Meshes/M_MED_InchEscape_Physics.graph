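# Physics asset for the InchEscape medium male player body cosmetic, set up for the jacket's back, left and right hanging panels.
# Each panel is a three-capsule dyn_ chain from a small MSTR_ root body constrained to the pelvis.
# Chain joints have tight angular limits, bending mostly on one swing axis, which differs between back and side panels.
# Large pelvis, thigh and calf capsules form the lower body, joined by free angular constraints.
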PhysicsAsset /BRCosmetics/Characters/Player/Male/Medium/Bodies/M_MED_InchEscape/Meshes/M_MED_InchEscape_Physics
bodies {
 MSTR_jacketBa_c [{shapeType=Capsule, length=0.505, radius=0.505, shapeName=MSTR_jacketBa_c_capsule, rotation=<rot 0 -90 0>}]
 MSTR_jacketD_l [{shapeType=Capsule, length=0.505, radius=0.505, shapeName=MSTR_jacketD_l_capsule, rotation=<rot 0 -90 0>}]
 MSTR_jacketD_r [{shapeType=Capsule, length=0.505, radius=0.505, shapeName=MSTR_jacketD_r_capsule, rotation=<rot 0 -90 0>}]
 calf_l [{shapeType=Capsule, length=25.401, radius=18, shapeName=calf_l_capsule, center=<-29.349 -5.0084 0>, rotation=<rot 90 0 0>}]
 calf_r [{shapeType=Capsule, length=24.2169, radius=18, shapeName=calf_r_capsule, center=<31.7111 2.7671 0>, rotation=<rot 90 0 0>}]
 dyn_jacketBa1_c [{shapeType=Capsule, length=6, radius=8, shapeName=dyn_jacketBa1_c_capsule, center=<4 0 0>, rotation=<rot 90 0 0>}]
 dyn_jacketBa2_c [{shapeType=Capsule, length=6, radius=8, shapeName=dyn_jacketBa2_c_capsule, center=<4 -4.5441 0>, rotation=<rot 90 0 0>}]
 dyn_jacketBa3_c [{shapeType=Capsule, length=15, radius=12, shapeName=dyn_jacketBa3_c_capsule, center=<4 -6.6345 0>}]
 dyn_jacketD1_l [{shapeType=Capsule, length=6, radius=8, shapeName=dyn_jacketD1_l_capsule, center=<4 0 0>, rotation=<rot 90 0 0>}]
 dyn_jacketD1_r [{shapeType=Capsule, length=6, radius=8, shapeName=dyn_jacketD1_r_capsule, center=<4 0 0>, rotation=<rot 90 0 0>}]
 dyn_jacketD2_l [{shapeType=Capsule, length=6, radius=8, shapeName=dyn_jacketD2_l_capsule, center=<4 0 5.7387>, rotation=<rot 90 0 0>}]
 dyn_jacketD2_r [{shapeType=Capsule, length=6, radius=8, shapeName=dyn_jacketD2_r_capsule, center=<4 0 4.6043>, rotation=<rot 90 0 0>}]
 dyn_jacketD3_l [{shapeType=Capsule, length=15, radius=12, shapeName=dyn_jacketD3_l_capsule, center=<4 0 6.4316>, rotation=<rot -0.0002 0.0001 -90>}]
 dyn_jacketD3_r [{shapeType=Capsule, length=15, radius=12, shapeName=dyn_jacketD3_r_capsule, center=<4 0.1643 5.2602>, rotation=<rot -0.0001 -0.0001 -75.9375>}]
 pelvis [{
   shapeType=Capsule
   length=14.2143
   radius=12.8954
   shapeName=pelvis_capsule
   center=<-6.7846 -1.0708 -0.0107>
   rotation=<rot -0.3786 -90 -0.0914>
  }]
 thigh_l [{shapeType=Capsule, length=29.5426, radius=13.9144, shapeName=thigh_l_capsule, center=<-21.7914 -0.8933 2.0168>, rotation=<rot 90 -0.0004 0>}]
 thigh_r [{shapeType=Capsule, length=29.8215, radius=13.2646, shapeName=thigh_r_capsule, center=<20.9786 0.9229 -1.7413>, rotation=<rot 90 0 0>}]
}
constraints [
 {
  bone1Name=thigh_r
  bone2Name=pelvis
  swing1LimitDegrees=45
  swing1Motion=Free
  swing2LimitDegrees=45
  swing2Motion=Free
  twistLimitDegrees=45
  twistMotion=Free
 }
 {
  bone1Name=thigh_l
  bone2Name=pelvis
  swing1LimitDegrees=45
  swing1Motion=Free
  swing2LimitDegrees=45
  swing2Motion=Free
  twistLimitDegrees=45
  twistMotion=Free
 }
 {bone1Name=dyn_jacketBa1_c, bone2Name=MSTR_jacketBa_c, swing1LimitDegrees=5, swing2LimitDegrees=1, twistLimitDegrees=1}
 {bone1Name=dyn_jacketBa2_c, bone2Name=dyn_jacketBa1_c, swing1LimitDegrees=20, swing2LimitDegrees=2, twistLimitDegrees=0}
 {bone1Name=dyn_jacketBa3_c, bone2Name=dyn_jacketBa2_c, swing1LimitDegrees=20, swing2LimitDegrees=5, twistLimitDegrees=1}
 {bone1Name=dyn_jacketD1_l, bone2Name=MSTR_jacketD_l, swing1LimitDegrees=1, swing2LimitDegrees=5, twistLimitDegrees=1}
 {bone1Name=dyn_jacketD2_l, bone2Name=dyn_jacketD1_l, swing1LimitDegrees=2, swing2LimitDegrees=20, twistLimitDegrees=0}
 {bone1Name=dyn_jacketD3_l, bone2Name=dyn_jacketD2_l, swing1LimitDegrees=5, swing2LimitDegrees=20, twistLimitDegrees=1}
 {bone1Name=dyn_jacketD1_r, bone2Name=MSTR_jacketD_r, swing1LimitDegrees=1, swing2LimitDegrees=5, twistLimitDegrees=1}
 {bone1Name=dyn_jacketD2_r, bone2Name=dyn_jacketD1_r, swing1LimitDegrees=2, swing2LimitDegrees=20, twistLimitDegrees=0}
 {bone1Name=dyn_jacketD3_r, bone2Name=dyn_jacketD2_r, swing1LimitDegrees=5, swing2LimitDegrees=20, twistLimitDegrees=1}
 {
  bone1Name=calf_l
  bone2Name=thigh_l
  swing1LimitDegrees=45
  swing1Motion=Free
  swing2LimitDegrees=45
  swing2Motion=Free
  twistLimitDegrees=45
  twistMotion=Free
 }
 {
  bone1Name=calf_r
  bone2Name=thigh_r
  swing1LimitDegrees=45
  swing1Motion=Free
  swing2LimitDegrees=45
  swing2Motion=Free
  twistLimitDegrees=45
  twistMotion=Free
 }
 {
  bone1Name=MSTR_jacketD_l
  bone2Name=pelvis
  swing1LimitDegrees=45
  swing1Motion=Free
  swing2LimitDegrees=45
  swing2Motion=Free
  twistLimitDegrees=45
  twistMotion=Free
 }
 {
  bone1Name=MSTR_jacketD_r
  bone2Name=pelvis
  swing1LimitDegrees=45
  swing1Motion=Free
  swing2LimitDegrees=45
  swing2Motion=Free
  twistLimitDegrees=45
  twistMotion=Free
 }
 {
  bone1Name=MSTR_jacketBa_c
  bone2Name=pelvis
  swing1LimitDegrees=45
  swing1Motion=Free
  swing2LimitDegrees=45
  swing2Motion=Free
  twistLimitDegrees=45
  twistMotion=Free
 }
]
